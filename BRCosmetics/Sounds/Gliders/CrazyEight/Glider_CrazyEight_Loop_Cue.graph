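# Sound cue for the looping audio of the CrazyEight cosmetic glider.
# A mixer plays two looping waves together, a slow loop weighted slightly louder than a fast loop, at a reduced overall volume.
# It uses the shared glider thrust attenuation, the cosmetic glider loop audio properties sheet and the glider landing sound class.
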SoundCue /BRCosmetics/Sounds/Gliders/CrazyEight/Glider_CrazyEight_Loop_Cue
root=Mixer_0
nodes [
 {id=Mixer_0, children [WavePlayer_4,WavePlayer_3], properties {inputVolume [0.8,0.6]}}
 {id=WavePlayer_3, properties {bLooping=true, soundWaveAssetPtr=&/Glider_CrazyEight_LoopFast02}}
 {id=WavePlayer_4, properties {bLooping=true, soundWaveAssetPtr=&/Glider_CrazyEight_LoopSlow02}}
]
properties {
 attenuationSettings=&Fort_Player_Glider_Thrust_Athena_Att
 audioPropertiesSheet=&APS_Cosmetic_Glider_Loop
 maxDistance=12500
 soundClassObject=&Fort_Gliders_Land
 volumeMultiplier=0.15
}
