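# Sound cue for the CrazyEight glider's opening sound in Battle Royale cosmetics.
# A root delay holds playback briefly, then a mixer layers the Open02 wave with a slightly boosted secondary layer wave that starts after a further fixed delay.
# It uses the shared glider open/close attenuation and concurrency, the cosmetic glider-open property sheet, and a glider sound class.
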SoundCue /BRCosmetics/Sounds/Gliders/CrazyEight/Glider_CrazyEight_Open_Cue
root=Delay_0
nodes [
 {id=Delay_0, children [Mixer_0], properties {delayMax=0.3, delayMin=0.3}}
 {id=Delay_1, children [WavePlayer_0], properties {delayMax=0.77, delayMin=0.77}}
 {id=Mixer_0, children [WavePlayer_1,Delay_1], properties {inputVolume [1,1.2]}}
 {id=WavePlayer_0, properties {soundWaveAssetPtr=&/Glider_CrazyEight_Open_Layer}}
 {id=WavePlayer_1, properties {soundWaveAssetPtr=&/Glider_CrazyEight_Open02}}
]
properties {
 attenuationSettings=/Game/Sounds/Attenuation/Fort_Player_Glider_OpenClose_Athena_Att
 audioPropertiesSheet=/BRCosmetics/Sounds/Settings/PropertySheets/Glider/APS_Cosmetic_Glider_Open
 maxDistance=10000
 soundClassObject=&Fort_Gliders_Land
 volumeMultiplier=1
 concurrencySet [&Athena_GliderOpenClose_Conc]
}
